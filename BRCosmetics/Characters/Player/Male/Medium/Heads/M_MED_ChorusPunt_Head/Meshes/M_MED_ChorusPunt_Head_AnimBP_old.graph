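# Animation Blueprint for the ChorusPunt male medium cosmetic head, marked old by its name, that derives a damped head rotation value.
# The update event reads the head bone's delta from reference pose relative to neck_02 on the owning skeletal mesh component.
# It scales roll, pitch and yaw by separate percentage variables, strongly reducing roll, and stores the result in headRot.
# No anim graph or other consumer of headRot appears here.
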AnimBlueprint /BRCosmetics/Characters/Player/Male/Medium/Heads/M_MED_ChorusPunt_Head/Meshes/M_MED_ChorusPunt_Head_AnimBP_old
target_skeleton=/Game/Characters/Player/Common/Fortnite_Base_Head/Fortnite_Base_Head_Export_Skeleton
graphs [EventGraph {nodes [
   Event BlueprintUpdateAnimation()
   n1 = GetOwnerSkeletalMeshComponent()
   n2 = BreakRotator(InRot=BreakTransform(InTransform=GetDeltaTransformFromRefPose(self=@n1, BoneName=head, BaseName=neck_02)).Rotation)
   n3 = Get Z_Percentage() -> (Z_Percentage:double)
   n4 = Get X_Percentage() -> (X_Percentage:double)
   n5 = Get Y_Percentage() -> (Y_Percentage:double)
   > Set headRot(headRot:Rotator=MakeRotator(
     Roll=float * float(A=@n2, B=@n4) {is=title K2Node_PromotableOperator}
     Pitch=float * float(A=@n2.Pitch, B=@n5)
     Yaw=float * float(A=@n2.Yaw, B=@n3)
    )) -> (Output_Get:Rotator)
   TryGetPawnOwner()
   Parent: BlueprintUpdateAnimation!(execute, DeltaTimeX:float=0) -> (then) {is=title K2Node_CallParentFunction}
  ]}]
variables [
 headRot {type=Rotator}
 X_Percentage {type=number, default=0.1}
 Y_Percentage {type=number, default=0.8}
 Z_Percentage {type=number, default=0.8}
]
